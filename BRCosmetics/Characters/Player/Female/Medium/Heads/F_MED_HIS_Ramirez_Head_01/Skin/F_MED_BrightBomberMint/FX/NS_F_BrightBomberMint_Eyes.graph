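# Niagara cosmetic effect giving the female BrightBomberMint head flaming green eyes, attached to two skeletal mesh sockets.
# EyeFlames and Glow sprites plus flickering Light particles use two spawn groups with mirrored offsets to cover both eyes.
# RibbonFlame alternates sockets by counter, rising with curl noise and socket-velocity drag, narrowing with distance and dying outside a small sphere.
# User.bisHeadHidden drives sprite visibility tags, disables lights and zeroes ribbon width.
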NiagaraSystem /BRCosmetics/Characters/Player/Female/Medium/Heads/F_MED_HIS_Ramirez_Head_01/Skin/F_MED_BrightBomberMint/FX/NS_F_BrightBomberMint_Eyes
dependencies {
 usedDataInterfaces [
  &NiagaraDataInterfaceCurve
  /Script/Niagara.NiagaraDataInterfaceSimpleCounter
  &NiagaraDataInterfaceVectorCurve
  &NiagaraDataInterfaceColorCurve
 ]
 usedDynamicInputs []
 usedModules [&SystemState]
}
emitters {
 EyeFlames {
  emitterUpdateScript [
   &EmitterState()
   &SpawnRate(SpawnRate=24, Use Spawn Probability=false)
  ]
  particleSpawnScript [
   /Niagara/Modules/Spawn/Initialization/InitializeParticle(
    Color=<rgba 0.3125 1 0.25 0.75>
    Lifetime=&UniformRangedFloat(Minimum=0.55, Maximum=0.65)
    Module.Write Color=true
    Module.Write Lifetime=true
    Module.Write Mass=false
    Module.Write Material Random=true
    Module.Write Position=false
    Module.Write Scale=false
    Module.Write SpriteRotation=true
    Module.Write SpriteSize=true
    Position=&SimulationPosition()
    Sprite Rotation=&UniformRangedFloat(Minimum=0, Maximum=360)
    Sprite Size=&Vector2DFromFloat(Value=&RandomRangeFloat(Minimum=6, Maximum=7))
   )
   &DynamicMaterialParameters(
    Index 0 Param 2=&UniformRangedFloat(Minimum=0, Maximum=1)
    Index 0 Param 3=1
    Index 0 Param 4=&UniformRangedFloat(Minimum=1, Maximum=1)
    Module.Param0WriteEnabled=false
    Module.Param1WriteEnabled=true
    Module.Param2WriteEnabled=false
    Module.Param3WriteEnabled=true
   ) {module=DynamicMaterialParameters002}
   &FN_SetSpawnGroup(Number of Spawn Groups=2)
   &CameraOffset(Camera Offset Amount=&RandomRangeFloat(Minimum=0.8, Maximum=1.5))
  ]
  particleUpdateScript [
   &ParticleState()
   &SkeletalMeshLocation(
    Bone / Socket Index=&SpawnGroup()
    Inherit Velocity=false
    Interpolate Mesh Sampling=false
    MaskSpawnGroup=false
    Mesh Sampling Type=::NewEnumerator5
    Sampled Position Offset=&MaskVectorBySpawnGroup(
     Masked Vector=&VectorFromCurve(VectorCurve=DataInterface() {propertyValues {XCurve {keys [<at 0 0.65>,<at 1 4.5>]}, YCurve {keys [<at 0 1.25>,<at 1 -3>]}, ZCurve {keys [<at 0 0.35>,<at 1 1.5>]}}})
     Passthrough Vector=&VectorFromCurve(VectorCurve=DataInterface() {propertyValues {XCurve {keys [<at 0 -0.65>,<at 1 -4.5>]}, YCurve {keys [<at 0 1.25>,<at 1 -3>]}, ZCurve {keys [<at 0 0.35>,<at 1 1.5>]}}})
    )
    Sampled Position Offset Coordinate Space:ENiagaraMeshTransforms=::NewEnumerator3
    Skeletal Mesh=$User.SkeletalMesh
    Socket Sampling Mode=::NewEnumerator5
   )
   &ScaleColor(
    Scale Mode=::NewEnumerator1
    Scale RGBA=&Vector4FromCurve(Vector4Curve=DataInterface() {propertyValues {
       AlphaCurve {keys [<at 0.0024 0>,<at 0.1965 2>,<at 0.5 3>,<at 1 0>]}
       BlueCurve {keys [<at 0.2365 0>,<at 0.65 0.098>,<at 0.85 0.0833>]}
       GreenCurve {keys [<at 0.2365 5>,<at 0.65 1>,<at 0.85 1>]}
       RedCurve {keys [<at 0.2365 0.4167>,<at 0.65 0.016>,<at 0.85 0>]}
      }})
    ScaleRGBA=true
   )
   /SetVariables(Particles.SubImageIndex=&FloatFromCurve(FloatCurve=DataInterface() {propertyValues {Curve {keys [<at 0 10>,<at 1 34>]}}})) {isSetParameters=true}
   &ScaleSpriteSize(Scale Factor=&Vector2DFromFloat(Value=&FloatFromCurve(FloatCurve=DataInterface() {propertyValues {Curve {keys [
          {arriveTangent=3.3052, time=0, value=0.35}
          {arriveTangent=-0.0353, leaveTangent=-0.0353, time=0.15, value=1.25}
          <at 0.5752 1.25>
          <at 1 0.65 RCIM_Cubic>
         ]}}})))
   &DynamicMaterialParameters(
    Index 0 Param 1=&FloatFromCurve(FloatCurve=DataInterface() {propertyValues {Curve {keys [{arriveTangent=2.7551, interpMode=RCIM_Cubic, leaveTangent=2.7551, tangentMode=RCTM_User, time=0, value=0}, <at 0.5 1>]}}})
    Index 0 Param 2=&RandomRangeFloat(Minimum=0.33, Maximum=1)
    Index 0 Param 3=&FloatFromCurve(FloatCurve=DataInterface() {propertyValues {Curve {keys [<at 0.8 1>,<at 1 2>]}}})
    Index 0 Param 4=1
    Module.Param0WriteEnabled=true
    Module.Param1WriteEnabled=true
    Module.Param2WriteEnabled=true
    Module.Param3WriteEnabled=true
   ) {module=DynamicMaterialParameters001}
   &SpriteRotationRate()
   /SetVariables(Particles.VisibilityTag=&MakeCustomIntFromBool(Bool=$User.bisHeadHidden, True INT=1, False INT=0)) {isSetParameters=true, module=SetVariables_#1}
  ]
  rendererClasses [&NiagaraSpriteRendererProperties]
  renderers [{rendererClass=&NiagaraSpriteRendererProperties}]
 }
 Glow {
  emitterUpdateScript [
   &EmitterState()
   &SpawnRate(SpawnRate=24, Use Spawn Probability=false)
  ]
  particleSpawnScript [
   &InitializeParticle(
    Color=<rgba 0.0625 0.75 0 1>
    Lifetime=&UniformRangedFloat(Minimum=0.45, Maximum=0.5)
    Module.Write Color=true
    Module.Write Lifetime=true
    Module.Write Mass=true
    Module.Write Material Random=true
    Module.Write Position=true
    Module.Write Scale=false
    Module.Write SpriteRotation=false
    Module.Write SpriteSize=true
    Position=&SimulationPosition()
    Sprite Size=&Vector2DFromFloat(Value=&UniformRangedFloat(Minimum=9, Maximum=9))
   )
   &DynamicMaterialParameters(
    Index 0 Param 1=&RandomRangeFloat(Minimum=0, Maximum=1)
    Module.Param0WriteEnabled=true
    Module.Param1WriteEnabled=false
    Module.Param2WriteEnabled=false
    Module.Param3WriteEnabled=false
   )
   &FN_SetSpawnGroup(Number of Spawn Groups=2)
   &CameraOffset(Camera Offset Amount=2.5)
  ]
  particleUpdateScript [
   &ParticleState()
   &SkeletalMeshLocation(
    Bone / Socket Index=&SpawnGroup()
    Inherit Velocity=false
    Interpolate Mesh Sampling=false
    MaskSpawnGroup=false
    Mesh Sampling Type=::NewEnumerator5
    Sampled Position Offset=&MaskVectorBySpawnGroup(Masked Vector=<0.4 0.5 -0.5>, Passthrough Vector=<-0.4 0.5 -0.5>)
    Sampled Position Offset Coordinate Space:ENiagaraMeshTransforms=::NewEnumerator3
    Skeletal Mesh=$User.SkeletalMesh
    Socket Sampling Mode=::NewEnumerator5
   ) {module=SkeletalMeshLocation001}
   &ScaleColor(
    Scale Alpha=&FloatFromCurve(FloatCurve=DataInterface() {propertyValues {Curve {keys [<at 0 0>,<at 0.25 1>,<at 0.8107 1>,<at 1 0>]}}})
    ScaleA=true
    ScaleRGB=true
   )
   &ScaleSpriteSize(Scale Factor=&Vector2DFromFloat(Value=&FloatFromCurve(FloatCurve=DataInterface() {propertyValues {Curve {keys [
          {arriveTangent=-0.0036, interpMode=RCIM_Cubic, leaveTangent=-0.0036, tangentMode=RCTM_User, time=0, value=1}
          {arriveTangent=-2.2201, interpMode=RCIM_Cubic, leaveTangent=-2.2201, tangentMode=RCTM_User, time=1, value=0.1}
         ]}}})))
   /SetVariables(Particles.VisibilityTag=&MakeCustomIntFromBool(Bool=$User.bisHeadHidden, True INT=1, False INT=0)) {isSetParameters=true, module=SetVariables_#1}
  ]
  rendererClasses [&NiagaraSpriteRendererProperties]
  renderers [{rendererClass=&NiagaraSpriteRendererProperties}]
 }
 Light {
  emitterUpdateScript [
   &EmitterState(Inactive Response=::NewEnumerator1, Life Cycle Mode=::NewEnumerator1, Loop Duration=2, UseLoopDelay=true)
   &SpawnBurst_Instantaneous(Spawn Count=2, Use Spawn Probability=false)
  ]
  particleSpawnScript [
   &InitializeParticle(
    Color=<rgba 0.0361 4 1.6878 1>
    Lifetime=2
    Module.Write Color=true
    Module.Write Lifetime=true
    Module.Write Mass=false
    Module.Write Material Random=true
    Module.Write Position=false
    Module.Write Scale=false
    Module.Write SpriteRotation=false
    Module.Write SpriteSize=true
    Position=&SimulationPosition()
    Sprite Size=<75 75>
   )
   &DynamicMaterialParameters(
    Index 0 Param 1=&UniformRangedFloat(Minimum=0.7, Maximum=0.9)
    Index 0 Param 2=&UniformRangedFloat(Minimum=0.05, Maximum=0.15)
    Index 0 Param 3=1
    Module.Param0WriteEnabled=true
    Module.Param1WriteEnabled=true
    Module.Param2WriteEnabled=false
    Module.Param3WriteEnabled=false
   ) {module=DynamicMaterialParameters002}
   &FN_SetSpawnGroup(Number of Spawn Groups=2)
  ]
  particleUpdateScript [
   &ParticleState()
   &SkeletalMeshLocation(
    Bone / Socket Index=&SpawnGroup()
    Inherit Velocity=false
    Interpolate Mesh Sampling=false
    MaskSpawnGroup=false
    Mesh Sampling Type=::NewEnumerator5
    Sampled Position Offset=&MaskVectorBySpawnGroup(Masked Vector=<1 1.5 0>, Passthrough Vector=<-1 1.5 0>)
    Sampled Position Offset Coordinate Space:ENiagaraMeshTransforms=::NewEnumerator3
    Skeletal Mesh=$User.SkeletalMesh
    Socket Sampling Mode=::NewEnumerator5
   )
   &SolveForcesAndVelocity(Clamp Velocity=false, Limit Acceleration=false)
   &ScaleColor(
    Scale RGB=&Multiply_Vector(
     A=&MakeVectorFromLinearColorRGB(Color=&LerpLinearColors(
       StartColor=<rgba 0.447 2 0.3059 1>
       EndColor=<rgba 0 2 0 1>
       LerpFactor=&FloatFromCurve(
        FloatCurve=DataInterface() {propertyValues {Curve {keys [<at 0 1>,<at 0.1069 0.0895>,<at 0.3206 0.7423>,<at 0.4907 0.1444>,<at 0.6992 0.8948>,<at 0.9061 0.1261>,<at 1 1>]}}}
        CurveIndex=$Emitter.NormalizedLoopAge
        Scale Curve=0.9
       )
      ))
     B=&VectorFromFloat(Value=&FloatFromCurve(
       FloatCurve=DataInterface() {propertyValues {Curve {keys [<at 0 1>,<at 0.0584 1.1216>,<at 0.0797 0.6691>,<at 0.1464 0.526>,<at 0.222 1.0327>,<at 0.2705 0.6844>,<at 0.3166 0.8798>,<at 0.3641 0.5477>,<at 0.4407 0.6405>,<at 0.4815 1.1358>,<at 0.5388 1.0006>,<at 0.58 0.5314>,<at 0.6479 0.8407>,<at 0.6937 0.7459>,<at 0.737 1.0896>,<at 0.8125 0.5703>,<at 0.8909 0.8732>,<at 0.9449 0.6186>,<at 1 1>]}}}
       CurveIndex=$Emitter.NormalizedLoopAge
      ))
    )
    ScaleA=true
    ScaleRGB=true
   )
   &FN_Light_v2(
    Light Exponent=1
    Light Radius=5
    LightEnabled:NiagaraBool=&Boolean_NotOperation(Bool=$User.bisHeadHidden)
    Scale Color=2
   )
   &ScaleSpriteSize(Scale Factor=&Vector2DFromFloat(Value=&FloatFromCurve(
      FloatCurve=DataInterface() {propertyValues {Curve {keys [<at 0 1>,<at 0.0562 0.9129>,<at 0.0748 1.1321>,<at 0.1235 1.3668>,<at 0.2098 0.8951>,<at 0.2589 1.1544>,<at 0.2987 1.0516>,<at 0.3338 1.2874>,<at 0.3749 1.3981>,<at 0.4815 0.8928>,<at 0.5082 1.113>,<at 0.5446 1.2763>,<at 0.6237 1.1354>,<at 0.6838 1.0572>,<at 0.7109 0.8761>,<at 0.7888 0.7005>,<at 0.8531 1.2103>,<at 0.9284 1.1544>,<at 1 1>]}}}
      CurveIndex=$Emitter.NormalizedLoopAge
     )))
  ]
  rendererClasses [&NiagaraLightRendererProperties]
  renderers [{rendererClass=&NiagaraLightRendererProperties, rendererIndex=0}]
 }
 RibbonFlame {
  emitterUpdateScript [
   &EmitterState()
   &SpawnRate(SpawnRate=75, Use Spawn Probability=false)
  ]
  particleSpawnScript [
   &InitializeRibbon(
    Color=<rgba 0.4067 4 0.08 0.5>
    Lifetime=&RandomRangeFloat(Minimum=0.2, Maximum=0.25)
    Module.Write Color=true
    Module.Write Lifetime=true
    Module.Write Mass=true
    Module.Write Material Random=true
    Module.Write Position=true
    Module.Write RibbonTwist=false
    Module.Write RibbonWidth=true
    Ribbon Width=&RandomRangeFloat(Minimum=3.5, Maximum=3.5)
   )
   /SetVariables(Particles.AlternateSockets:NiagaraInt32=&ModuloInt(
     A=/Niagara/DynamicInputs/DataInterface/GenerateIntFromCounter(SimpleCounter:NiagaraDataInterfaceSimpleCounter=DataInterface() {propertyValues {GpuSyncMode=None, InitialValue=0}})
     B=2
    )) {isSetParameters=true}
   &SkeletalMeshLocation(
    Bone / Socket Index=$Particles.AlternateSockets
    Inherit Velocity=true
    MaskSpawnGroup=false
    Mesh Sampling Type=::NewEnumerator5
    Sampled Position Offset Coordinate Space:ENiagaraMeshTransforms=::NewEnumerator3
    Skeletal Mesh=$User.SkeletalMesh
    Socket Sampling Mode=::NewEnumerator5
    VelocitySpeedLimit=false
   )
   /SetVariables(Particles.RibbonLinkOrder:NiagaraFloat=&MakeFloatFromInt(Int=&GenerateIntFromCounter(SimpleCounter:NiagaraDataInterfaceSimpleCounter=DataInterface() {propertyValues {GpuSyncMode=None, InitialValue=0}}))) {isSetParameters=true}
  ]
  particleUpdateScript [
   &ParticleState()
   &SkeletalMeshLocation(
    Apply To Particle Position:NiagaraBool=&SetBoolByFloatComparison(A=$Particles.NormalizedAge, B=0.2, Comparison Type=::NewEnumerator6)
    Bone / Socket Index=$Particles.Initial.AlternateSockets
    Inherit Velocity=false
    Interpolate Mesh Sampling=false
    MaskSpawnGroup=false
    Mesh Sampling Type=::NewEnumerator5
    Sampled Position Offset=<0 0 -0.5>
    Skeletal Mesh=$User.SkeletalMesh
    Socket Sampling Mode=::NewEnumerator5
   ) {module=SkeletalMeshLocation001}
   &GravityForce(Coordinate Space=::Local, Gravity=&RandomRangeVector(Minimum=<-5 -5 160>, Maximum=<5 5 190>))
   /Niagara/Modules/Update/Forces/CurlNoiseForce(Noise Strength=20, Pan Noise=false, Random Seed:NiagaraInt32=&RandomRangeInt(Minimum=0, Maximum=100))
   &Drag(
    Drag=&Multiply_Float(A=/Niagara/DynamicInputs/TypeConversions/VectorLengthSafe(Vector:Vector3f=$Output.SkeletalMeshLocation001.SampledVelocity), B=0.9)
    Use Linear Drag=true
    Use Rotational Drag=false
   )
   &SolveForcesAndVelocity(Clamp Velocity=false, Limit Acceleration=false)
   &ScaleColor(
    Scale Alpha=&FloatFromCurve(FloatCurve=DataInterface() {propertyValues {Curve {keys [<at 0 1 RCIM_Cubic>, {arriveTangent=-1, interpMode=RCIM_Cubic, leaveTangent=-1, time=0.2, value=1}, <at 1 0 RCIM_Cubic>]}}})
    ScaleA=true
    ScaleRGB=true
   )
   /SetVariables(Particles.DistanceScale:NiagaraFloat=&Add_Float(
     A=1
     B=&ClampFloat(
      Float=&Multiply_Float(
       A=-0.1
       B=&VectorLength(Vector=&Subtract_Vector(
         A=Linked() {linkedVariable {name=Particles.Position, type=NiagaraPosition}}
         B=Linked() {linkedVariable {name=Particles.Initial.Position, type=NiagaraPosition}}
        ))
      )
      Min=-0.6
      Max=0
     )
    )) {isSetParameters=true}
   &KillParticlesInVolume(Invert Volume=true, Sphere Radius=7, Volume Origin:Vector3f=$Output.SkeletalMeshLocation001.SampledPosition)
   &ScaleRibbonWidth(Ribbon Width Scale=&FloatFromCurve(
     FloatCurve=DataInterface() {propertyValues {Curve {keys [
         {arriveTangent=6.0445, interpMode=RCIM_Cubic, leaveTangent=6.0445, tangentMode=RCTM_User, time=0, value=0.5}
         {arriveTangent=-0.5, interpMode=RCIM_Cubic, leaveTangent=-0.5, time=0.2036, value=1}
         <at 1 0>
        ]}}}
     Scale Curve=&MakeCustomFloatFromBool(
      Bool=$User.bisHeadHidden
      True Float=0
      False Float=&MakeCustomFloatFromBool(
       Bool=&SetBoolByFloatComparison(A=$Particles.DistanceScale, B=0.5)
       True Float=1
       False Float=$Particles.DistanceScale
      )
     )
    ))
   &DynamicMaterialParameters(
    Index 0 Param 1=$Particles.VelocityScale
    Module.Param0WriteEnabled=true
    Module.Param1WriteEnabled=true
    Module.Param2WriteEnabled=true
    Module.Param3WriteEnabled=true
   )
  ]
  rendererClasses [&NiagaraRibbonRendererProperties]
  renderers [{rendererClass=&NiagaraRibbonRendererProperties}]
 }
}
summary {
 systemName=NS_F_BrightBomberMint_Eyes
 emitters [
  {emitterName=Light, rendererClasses [&NiagaraLightRendererProperties]}
  {emitterName=Glow, rendererClasses [&NiagaraSpriteRendererProperties]}
  {emitterName=RibbonFlame, rendererClasses [&NiagaraRibbonRendererProperties]}
  {emitterName=EyeFlames, rendererClasses [&NiagaraSpriteRendererProperties]}
 ]
 userVariables [
  User.bisHeadHidden {type=NiagaraBool}
  User.IsSkydiving {type=NiagaraBool}
  User.SkeletalMesh {
   type=NiagaraDataInterfaceSkeletalMesh
   defaultValue {
    struct=NiagaraExt_VariableValue_DataInterface
    value {dataInterface=&self:NiagaraDataInterfaceSkeletalMesh_0}
   }
  }
 ]
}
